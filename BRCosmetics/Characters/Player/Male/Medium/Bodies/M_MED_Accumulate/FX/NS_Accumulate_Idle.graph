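# Idle Niagara effect for the Accumulate character body, growing mesh drips and drops from skeletal mesh bones or sockets selected by spawn group.
# Upwards, Downward and Hero drips are spring-driven pendulums pivoting on sampled points, passing pivot and tip positions to dynamic material parameters.
# Hero and Secondary drops accelerate up or down per group while partly inheriting source movement.
# Hand and foot hidden flags zero mesh scale for affected particles.
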NiagaraSystem /BRCosmetics/Characters/Player/Male/Medium/Bodies/M_MED_Accumulate/FX/NS_Accumulate_Idle
dependencies {
 usedDataInterfaces [&NiagaraDataInterfaceCurve, &NiagaraDataInterfaceVectorCurve, &NiagaraDataInterfaceSkeletalMesh]
 usedDynamicInputs []
 usedModules [
  &self:SystemSpawnScript.SystemScriptSource.SystemScriptGraph.NiagaraNodeAssignment_0.SetVariables_#1
  &SystemState
 ]
}
emitters {
 Downward_Drips {
  emitterUpdateScript [
   &EmitterState()
   &SpawnBurst_Instantaneous(Spawn Count=1, Spawn Probability=HlslExpression() {hlslExpression="Emitter.LoopCount == 0"}, Use Spawn Probability=false)
  ]
  particleSpawnScript [
   &FN_SetSpawnGroup(Number of Spawn Groups=2)
   /SetVariables(
    Particles.PendulumDrag=1
    Particles.PendulumLengthScale:NiagaraFloat=&RandomRangeFloat(Minimum=0.335, Maximum=0.375)
    Particles.PendulumMaskFalloff=1
    Particles.PendulumMeshWidth:NiagaraFloat=&RandomRangeFloat(Minimum=0.1, Maximum=0.175)
    Particles.PendulumRigidity=0.06
   ) {isSetParameters=true, module=SetVariables_#2}
   /Niagara/Modules/Spawn/Initialization/InitializeParticle(
    Lifetime=&UniformRangedFloat(Minimum=3, Maximum=6)
    Mesh Scale=&MakeVector(X=$Particles.PendulumMeshWidth, Y=$Particles.PendulumMeshWidth, Z=$Particles.PendulumLengthScale)
    Module.Write Color=true
    Module.Write Lifetime=true
    Module.Write Mass=true
    Module.Write Material Random=true
    Module.Write Position=false
    Module.Write Scale=true
    Module.Write SpriteRotation=false
    Module.Write SpriteSize=true
   )
   &CameraOffset(Camera Offset Amount=1)
  ]
  particleUpdateScript [
   &ParticleState()
   &SkeletalMeshLocation(
    Apply To Particle Position=false
    Bone / Socket Index=&Add_Integer(A=$Particles.SpawnGroup, B=3)
    Bone Sampling Mode=::NewEnumerator4
    Inherit Velocity=false
    MaskSpawnGroup=false
    Skeletal Mesh=$System.Skeletal Mesh
   )
   /SetVariables(Particles.MeshOrigin:Vector3f=&Add_Vector(
     A=$Output.SkeletalMeshLocation.SampledPosition
     B=&RotateVectorByQuaternion(
      Vector To Rotate=&RandomRangeVector(Minimum=<-1.25 -1.25 -1.25>, Maximum=<1.25 1.25 1.25>)
      Quaternion=$Output.SkeletalMeshLocation.SampledRotation
     )
    )) {isSetParameters=true, module=SetVariables_#3}
   &Drag(Drag=$Particles.PendulumDrag, Use Linear Drag=true, Use Rotational Drag=false)
   &GravityForce()
   &PendulumSetup(
    Module.Enable Spring Driver=true
    Pendulum Length:NiagaraFloat=&Multiply_Float(A=100, B=$Particles.PendulumLengthScale)
    Pendulum Pivot:Vector3f=$Particles.MeshOrigin
    "Spring Driven Constraint (Tightness)"=12
   )
   &SolveForcesAndVelocity(Clamp Velocity=false, Limit Acceleration=false)
   &PendulumConstraint()
   &CalculateAccurateVelocity()
   /SetVariables(Particles.MeshFacing:Vector3f=&NormalizeVector(Vector To Normalize=&Subtract_Vector(A=$Particles.MeshOrigin, B=Linked() {linkedVariable {name=Particles.Position, type=NiagaraPosition}}))) {isSetParameters=true, module=SetVariables_#4}
   &self:Pendulum_Facing(
    Bone Quat:Quat4f=$Output.SkeletalMeshLocation.SampledRotation
    FromVector:Vector3f=<0 0 1>
    ToVector:Vector3f=$Particles.MeshFacing
    X Vector:Vector3f=<1 0 0>
    Y Vector:Vector3f=<0 0 0>
   ) {module=Pendulum_Facing}
   &DynamicMaterialParameters(
    Index 1 Param 3=1
    Index 1 Param 4=&FloatFromCurve(FloatCurve=DataInterface() {propertyValues {Curve {keys [
         {arriveTangent=3.6645, interpMode=RCIM_Cubic, leaveTangent=3.6645, tangentMode=RCTM_User, time=0.1, value=-0.1}
         {interpMode=RCIM_Cubic, leaveTangent=2.3573, time=0.9, value=2}
        ]}}})
    Index 2 Param 1=HlslExpression() {hlslExpression=Particles.MeshOrigin.x}
    Index 2 Param 2=HlslExpression() {hlslExpression=Particles.MeshOrigin.y}
    Index 2 Param 3=HlslExpression() {hlslExpression=Particles.MeshOrigin.z}
    Index 2 Param 4=$Particles.PendulumMaskFalloff
    Index 3 Param 1=&MakeFloatFromVector(Vector=Linked() {linkedVariable {name=Particles.Position, type=NiagaraPosition}}, Channel=::NewEnumerator0)
    Index 3 Param 2=&MakeFloatFromVector(Vector=Linked() {linkedVariable {name=Particles.Position, type=NiagaraPosition}}, Channel=::NewEnumerator1)
    Index 3 Param 3=&MakeFloatFromVector(Vector=Linked() {linkedVariable {name=Particles.Position, type=NiagaraPosition}}, Channel=::NewEnumerator2)
    Index 3 Param 4=$Particles.PendulumRigidity
    Module.Param0WriteEnabled1=true
    Module.Param0WriteEnabled2=true
    Module.Param0WriteEnabled3=true
    Module.Param1WriteEnabled1=true
    Module.Param1WriteEnabled2=true
    Module.Param1WriteEnabled3=true
    Module.Param2WriteEnabled1=true
    Module.Param2WriteEnabled2=true
    Module.Param2WriteEnabled3=true
    Module.Param3WriteEnabled1=true
    Module.Param3WriteEnabled2=true
    Module.Param3WriteEnabled3=true
    Write Parameter Index 0=false
    Write Parameter Index 1=true
    Write Parameter Index 2=true
    Write Parameter Index 3=true
   )
   &ScaleMeshSize(Scale Factor=&Multiply_VectorByFloat(
     Vector=&MakeVector(
      X=&FloatFromCurve(FloatCurve=DataInterface() {propertyValues {Curve {keys [
           <at 0 0 RCIM_Cubic>
           {arriveTangent=1.0833, interpMode=RCIM_Cubic, leaveTangent=1.0833, time=0.05, value=0.9}
           {arriveTangent=-0.6176, interpMode=RCIM_Cubic, leaveTangent=-0.6176, time=0.6, value=0.65}
           {interpMode=RCIM_Cubic, leaveTangent=-4.2857, time=0.9, value=0.375}
           {interpMode=RCIM_Cubic, leaveTangent=-0.6579, time=0.91, value=0}
          ]}}})
      Y=&FloatFromCurve(FloatCurve=DataInterface() {propertyValues {Curve {keys [
           <at 0 0 RCIM_Cubic>
           {arriveTangent=1.0833, interpMode=RCIM_Cubic, leaveTangent=1.0833, time=0.05, value=0.9}
           {arriveTangent=-0.6176, interpMode=RCIM_Cubic, leaveTangent=-0.6176, time=0.6, value=0.65}
           {interpMode=RCIM_Cubic, leaveTangent=-4.2857, time=0.9, value=0.375}
           {interpMode=RCIM_Cubic, leaveTangent=-0.6579, time=0.91, value=0}
          ]}}})
      Z=&FloatFromCurve(FloatCurve=DataInterface() {propertyValues {Curve {keys [
           {arriveTangent=1.8796, interpMode=RCIM_Cubic, time=0, value=0}
           {arriveTangent=-4.7929, interpMode=RCIM_Cubic, leaveTangent=-4.7929, tangentMode=RCTM_User, time=0.625, value=0.85}
           {arriveTangent=-1.7696, interpMode=RCIM_Cubic, leaveTangent=-1.7696, tangentMode=RCTM_User, time=0.95, value=0}
          ]}}})
     )
     Float=&MakeCustomFloatFromBool(Bool=$User.bisHandHidden, True Float=0, False Float=1)
    ))
   /SetVariables(
    Particles.Drop Mesh Size:Vector3f=&VectorFromCurve(
     VectorCurve=DataInterface() {propertyValues {
       XCurve {keys [
         <at 0 1>
         {arriveTangent=-18.2164, interpMode=RCIM_Cubic, leaveTangent=-18.2164, tangentMode=RCTM_User, time=0.85, value=1}
         {arriveTangent=-2.286, interpMode=RCIM_Cubic, leaveTangent=-2.286, tangentMode=RCTM_User, time=0.8575, value=0}
        ]}
       YCurve {keys [
         <at 0 1>
         {arriveTangent=-18.2164, interpMode=RCIM_Cubic, leaveTangent=-18.2164, tangentMode=RCTM_User, time=0.85, value=1}
         {arriveTangent=-2.286, interpMode=RCIM_Cubic, leaveTangent=-2.286, tangentMode=RCTM_User, time=0.8575, value=0}
        ]}
       ZCurve {keys [
         <at 0.1 0.675>
         <at 0.3 0.8>
         <at 0.6 0.4>
         <at 0.8 0.8>
         {interpMode=RCIM_Cubic, leaveTangent=-5.2459, time=0.85, value=0.675}
         <at 0.8575 0>
        ]}
      }}
     Scale Curve=&VectorFromFloat(Value=&FloatFromCurve(
       FloatCurve=DataInterface() {propertyValues {Curve {keys [<at 0.46 0>,<at 0.51 2>,<at 0.6 1>,<at 0.7 0.55>,<at 0.8 1.2>,<at 0.9 0.7>,<at 1 1>]}}}
       Scale Curve=&RandomRangeFloat(Minimum=0.04, Maximum=0.05)
      ))
    )
    Particles.Drop Position:Vector3f=&Lerp_Vector(
     A=$Particles.MeshOrigin
     B=&Add_Vector(
      A=Linked() {linkedVariable {name=Particles.Position, type=NiagaraPosition}}
      B=&Multiply_VectorByFloat(
       Vector=&Multiply_VectorByFloat(
        Vector=&RandomRangeVector(Minimum=<-4 -4 -4>, Maximum=<4 4 4>)
        Float=&Sine(Normalized Angle=$Emitter.Age, Period=1, Scale=1, Bias=0)
       )
       Float=&FloatFromCurve(FloatCurve=DataInterface() {propertyValues {Curve {keys [
            {arriveTangent=0.7009, interpMode=RCIM_Cubic, leaveTangent=0.7009, tangentMode=RCTM_User, time=0.55, value=1}
            {arriveTangent=0.7952, interpMode=RCIM_Cubic, leaveTangent=0.7952, tangentMode=RCTM_User, time=0.9, value=0}
           ]}}})
      )
     )
     Alpha=&FloatFromCurve(FloatCurve=DataInterface() {propertyValues {Curve {keys [
          {arriveTangent=-0.2799, interpMode=RCIM_Cubic, leaveTangent=-0.2799, tangentMode=RCTM_User, time=0.275, value=0}
          {interpMode=RCIM_Cubic, tangentMode=RCTM_Break, time=0.6, value=0.35}
          {arriveTangent=5.35, interpMode=RCIM_Cubic, leaveTangent=15.5091, tangentMode=RCTM_Break, time=0.8, value=0.45}
          {arriveTangent=0.846, interpMode=RCIM_Cubic, leaveTangent=0.846, tangentMode=RCTM_User, time=1, value=2}
         ]}}})
    )
   ) {enabled=false, isSetParameters=true, module=SetVariables_#5}
  ]
  rendererClasses [&NiagaraMeshRendererProperties]
  renderers [
   {rendererClass=&NiagaraMeshRendererProperties, rendererIndex=0}
   {rendererClass=&NiagaraMeshRendererProperties, rendererIndex=1}
  ]
 }
 Hero_Drips {
  emitterUpdateScript [
   &EmitterState()
   &SpawnBurst_Instantaneous(Spawn Count=2, Spawn Probability=HlslExpression() {hlslExpression="Emitter.LoopCount == 0"}, Use Spawn Probability=false)
  ]
  particleSpawnScript [
   &FN_SetSpawnGroup(Number of Spawn Groups=2)
   /SetVariables(
    Particles.PendulumDrag=1
    Particles.PendulumLengthScale:NiagaraFloat=&MaskFloatBySpawnGroup(
     Masked Float=&RandomRangeFloat(Minimum=0.35, Maximum=0.5)
     Passthrough Float=&RandomRangeFloat(Minimum=0.375, Maximum=0.415)
    )
    Particles.PendulumMaskFalloff=1
    Particles.PendulumMeshWidth:NiagaraFloat=&RandomRangeFloat(Minimum=0.1, Maximum=0.125)
    Particles.PendulumRigidity=0.06
   ) {isSetParameters=true, module=SetVariables_#2}
   &InitializeParticle(
    Lifetime=&UniformRangedFloat(Minimum=4, Maximum=5)
    Mesh Scale=&MakeVector(X=$Particles.PendulumMeshWidth, Y=$Particles.PendulumMeshWidth, Z=$Particles.PendulumLengthScale)
    Module.Write Color=true
    Module.Write Lifetime=true
    Module.Write Mass=true
    Module.Write Material Random=true
    Module.Write Position=false
    Module.Write Scale=true
    Module.Write SpriteRotation=false
    Module.Write SpriteSize=true
   )
   &CameraOffset(Camera Offset Amount=2)
  ]
  particleUpdateScript [
   &ParticleState()
   &SkeletalMeshLocation(
    Apply To Particle Position=false
    Bone / Socket Index=&Add_Integer(A=$Particles.SpawnGroup, B=5)
    Bone Sampling Mode=::NewEnumerator4
    Inherit Velocity=false
    MaskSpawnGroup=false
    Skeletal Mesh=$System.Skeletal Mesh
   )
   /SetVariables(Particles.MeshOrigin:Vector3f=&Add_Vector(
     A=$Output.SkeletalMeshLocation.SampledPosition
     B=&RotateVectorByQuaternion(
      Vector To Rotate=&RandomRangeVector(Minimum=<-0.5 -0.5 -0.5>, Maximum=<0.5 0.5 0.5>)
      Quaternion=$Output.SkeletalMeshLocation.SampledRotation
     )
    )) {isSetParameters=true, module=SetVariables_#3}
   &Drag(Drag=$Particles.PendulumDrag, Use Linear Drag=true, Use Rotational Drag=false)
   &GravityForce()
   &PendulumSetup(
    Module.Enable Spring Driver=true
    Pendulum Length:NiagaraFloat=&Multiply_Float(A=100, B=$Particles.PendulumLengthScale)
    Pendulum Pivot:Vector3f=$Particles.MeshOrigin
    Pendulum Rest Axis:Vector3f=&MaskVectorBySpawnGroup(Masked Vector=<0 0 1>, Passthrough Vector=<0 0 -1>)
    "Spring Driven Constraint (Tightness)"=10
   )
   /SetVariables(Particles.MeshFacing:Vector3f=&NormalizeVector(Vector To Normalize=&Subtract_Vector(A=$Particles.MeshOrigin, B=Linked() {linkedVariable {name=Particles.Position, type=NiagaraPosition}}))) {isSetParameters=true, module=SetVariables_#4}
   &self:Pendulum_Facing(
    Bone Quat:Quat4f=$Output.SkeletalMeshLocation.SampledRotation
    FromVector:Vector3f=&MaskVectorBySpawnGroup(Masked Vector=<0 0 -1>, Passthrough Vector=<0 0 1>)
    ToVector:Vector3f=$Particles.MeshFacing
    X Vector:Vector3f=&MaskVectorBySpawnGroup(Masked Vector=<-1 0 0>, Passthrough Vector=<1 0 0>)
    Y Vector:Vector3f=<0 0 0>
   ) {module=Pendulum_Facing}
   &SolveForcesAndVelocity(Clamp Velocity=false, Limit Acceleration=false)
   &PendulumConstraint()
   &CalculateAccurateVelocity()
   &DynamicMaterialParameters(
    Index 1 Param 3=1
    Index 1 Param 4=&FloatFromCurve(FloatCurve=DataInterface() {propertyValues {Curve {keys [
         {arriveTangent=3.6645, interpMode=RCIM_Cubic, leaveTangent=3.6645, tangentMode=RCTM_User, time=0.1, value=-0.1}
         {interpMode=RCIM_Cubic, leaveTangent=2.3573, time=0.9, value=2}
        ]}}})
    Index 2 Param 1=HlslExpression() {hlslExpression=Particles.MeshOrigin.x}
    Index 2 Param 2=HlslExpression() {hlslExpression=Particles.MeshOrigin.y}
    Index 2 Param 3=HlslExpression() {hlslExpression=Particles.MeshOrigin.z}
    Index 2 Param 4=$Particles.PendulumMaskFalloff
    Index 3 Param 1=&MakeFloatFromVector(Vector=Linked() {linkedVariable {name=Particles.Position, type=NiagaraPosition}}, Channel=::NewEnumerator0)
    Index 3 Param 2=&MakeFloatFromVector(Vector=Linked() {linkedVariable {name=Particles.Position, type=NiagaraPosition}}, Channel=::NewEnumerator1)
    Index 3 Param 3=&MakeFloatFromVector(Vector=Linked() {linkedVariable {name=Particles.Position, type=NiagaraPosition}}, Channel=::NewEnumerator2)
    Index 3 Param 4=$Particles.PendulumRigidity
    Module.Param0WriteEnabled1=true
    Module.Param0WriteEnabled2=true
    Module.Param0WriteEnabled3=true
    Module.Param1WriteEnabled1=true
    Module.Param1WriteEnabled2=true
    Module.Param1WriteEnabled3=true
    Module.Param2WriteEnabled1=true
    Module.Param2WriteEnabled2=true
    Module.Param2WriteEnabled3=true
    Module.Param3WriteEnabled1=true
    Module.Param3WriteEnabled2=true
    Module.Param3WriteEnabled3=true
    Write Parameter Index 0=false
    Write Parameter Index 1=true
    Write Parameter Index 2=true
    Write Parameter Index 3=true
   )
   &ScaleMeshSize(Scale Factor=&Multiply_VectorByFloat(
     Vector=&MakeVector(
      X=&FloatFromCurve(FloatCurve=DataInterface() {propertyValues {Curve {keys [
           <at 0 0 RCIM_Cubic>
           {arriveTangent=2.5, interpMode=RCIM_Cubic, leaveTangent=2.5, time=0.05, value=0.9}
           {arriveTangent=-0.7353, interpMode=RCIM_Cubic, leaveTangent=-0.7353, time=0.6, value=0.65}
           {arriveTangent=-4.2857, interpMode=RCIM_Cubic, leaveTangent=-4.2857, time=0.9, value=0.375}
           {interpMode=RCIM_Cubic, leaveTangent=-0.6579, time=0.95, value=0}
          ]}}})
      Y=&FloatFromCurve(FloatCurve=DataInterface() {propertyValues {Curve {keys [
           <at 0 0 RCIM_Cubic>
           {arriveTangent=2.5, interpMode=RCIM_Cubic, leaveTangent=2.5, time=0.05, value=0.9}
           {arriveTangent=-0.7353, interpMode=RCIM_Cubic, leaveTangent=-0.7353, time=0.6, value=0.65}
           {arriveTangent=-4.2857, interpMode=RCIM_Cubic, leaveTangent=-4.2857, time=0.9, value=0.375}
           {interpMode=RCIM_Cubic, leaveTangent=-0.6579, time=0.95, value=0}
          ]}}})
      Z=&FloatFromCurve(FloatCurve=DataInterface() {propertyValues {Curve {keys [
           {arriveTangent=1.8796, interpMode=RCIM_Cubic, time=0, value=0}
           {arriveTangent=-4.7929, interpMode=RCIM_Cubic, leaveTangent=-4.7929, tangentMode=RCTM_User, time=0.625, value=0.85}
           {arriveTangent=-1.7696, interpMode=RCIM_Cubic, leaveTangent=-1.7696, tangentMode=RCTM_User, time=0.95, value=0}
          ]}}})
     )
     Float=&MakeCustomFloatFromBool(Bool=$User.bisHandHidden, True Float=0, False Float=1)
    ))
   /SetVariables(
    Particles.Drop Mesh Size:Vector3f=&VectorFromCurve(
     VectorCurve=DataInterface() {propertyValues {
       XCurve {keys [
         <at 0 1>
         {arriveTangent=-18.2164, interpMode=RCIM_Cubic, leaveTangent=-18.2164, tangentMode=RCTM_User, time=0.85, value=1}
         {arriveTangent=-2.286, interpMode=RCIM_Cubic, leaveTangent=-2.286, tangentMode=RCTM_User, time=0.8575, value=0}
        ]}
       YCurve {keys [
         <at 0 1>
         {arriveTangent=-18.2164, interpMode=RCIM_Cubic, leaveTangent=-18.2164, tangentMode=RCTM_User, time=0.85, value=1}
         {arriveTangent=-2.286, interpMode=RCIM_Cubic, leaveTangent=-2.286, tangentMode=RCTM_User, time=0.8575, value=0}
        ]}
       ZCurve {keys [
         <at 0.1 0.675>
         <at 0.3 0.8>
         <at 0.6 0.4>
         <at 0.8 0.8>
         {interpMode=RCIM_Cubic, leaveTangent=-5.2459, time=0.85, value=0.675}
         <at 0.8575 0>
        ]}
      }}
     Scale Curve=&VectorFromFloat(Value=&FloatFromCurve(
       FloatCurve=DataInterface() {propertyValues {Curve {keys [<at 0.46 0>,<at 0.51 2>,<at 0.6 1>,<at 0.7 0.55>,<at 0.8 1.2>,<at 0.9 0.7>,<at 1 1>]}}}
       Scale Curve=&RandomRangeFloat(Minimum=0.04, Maximum=0.05)
      ))
    )
    Particles.Drop Position:NiagaraPosition=$Particles.Position
   ) {enabled=false, isSetParameters=true, module=SetVariables_#5}
  ]
  rendererClasses [&NiagaraMeshRendererProperties]
  renderers [
   {rendererClass=&NiagaraMeshRendererProperties, rendererIndex=0}
   {rendererClass=&NiagaraMeshRendererProperties, rendererIndex=1}
  ]
 }
 Hero_Drops {
  emitterUpdateScript [
   &EmitterState()
   &SpawnRate(Use Spawn Probability=true)
  ]
  particleSpawnScript [
   &FN_SetSpawnGroup(Number of Spawn Groups=2)
   &InitializeParticle(
    Lifetime=&UniformRangedFloat(Minimum=0.85, Maximum=1.2)
    Mesh Scale=&VectorFromFloat(Value=&RandomRangeFloat(Minimum=0.04, Maximum=0.05))
    Module.Write Color=true
    Module.Write Lifetime=true
    Module.Write Mass=true
    Module.Write Material Random=true
    Module.Write Position=false
    Module.Write Scale=true
    Module.Write SpriteRotation=false
    Module.Write SpriteSize=true
   )
   &SkeletalMeshLocation(
    Bone / Socket Index=&Add_Integer(A=$Particles.SpawnGroup, B=5)
    Bone Sampling Mode=::NewEnumerator4
    Inherit Velocity=false
    MaskSpawnGroup=false
    Sampled Position Offset Coordinate Space:ENiagaraMeshTransforms=::NewEnumerator3
    Skeletal Mesh=$System.Skeletal Mesh
   )
   &SphereLocation(
    MaskSpawnGroup=false
    Override Seed=false
    Sphere Radius=2
    Surface Only=false
    UseNonUniformScale=false
    UseOffset=false
    UseSphereOrientation=false
   )
   &AddVelocity(Velocity=&Multiply_VectorByFloat(
     Vector=&RandomRangeVector(Minimum=<-3 -3 15>, Maximum=<3 3 35>)
     Float=&MaskFloatBySpawnGroup(Masked Float=1, Passthrough Float=-1)
    ))
  ]
  particleUpdateScript [
   &ParticleState()
   &ScaleVelocity(Velocity Scale=&MakeVector(
     X=1
     Y=1
     Z=&FloatFromCurve(FloatCurve=DataInterface() {propertyValues {Curve {keys [
          {arriveTangent=7.4745, interpMode=RCIM_Cubic, leaveTangent=7.4745, tangentMode=RCTM_User, time=0, value=0}
          {arriveTangent=14.58, interpMode=RCIM_Cubic, leaveTangent=14.58, tangentMode=RCTM_User, time=1, value=4}
         ]}}})
    ))
   &InheritSourceMovement(
    Applied Position Scale=&Multiply_VectorByFloat(
     Vector=<1 1 1>
     Float=&FloatFromCurve(FloatCurve=DataInterface() {propertyValues {Curve {keys [<at 0.4 1>,<at 1 0>]}}})
    )
    ApplyForce=false
    ApplyPosition=true
    ApplyPositionFalloff=false
   )
   &SolveForcesAndVelocity(Clamp Velocity=false, Limit Acceleration=false)
   &ScaleMeshSize(Scale Factor=&Multiply_VectorByFloat(
     Vector=&VectorFromFloat(Value=&FloatFromCurve(FloatCurve=DataInterface() {propertyValues {Curve {keys [
           {arriveTangent=5.6263, interpMode=RCIM_Cubic, leaveTangent=5.6263, tangentMode=RCTM_User, time=0, value=0}
           {interpMode=RCIM_Cubic, tangentMode=RCTM_User, time=0.5, value=1}
           {arriveTangent=-4.9972, interpMode=RCIM_Cubic, leaveTangent=-4.9972, tangentMode=RCTM_User, time=0.9, value=0}
          ]}}}))
     Float=&MakeCustomFloatFromBool(Bool=$User.bisHandHidden, True Float=0, False Float=1)
    ))
  ]
  rendererClasses [&NiagaraMeshRendererProperties]
  renderers [{rendererClass=&NiagaraMeshRendererProperties, rendererIndex=0}]
 }
 Secondary_Drops {
  emitterUpdateScript [
   &EmitterState()
   &SpawnRate(SpawnRate=3, Use Spawn Probability=false)
  ]
  particleSpawnScript [
   &FN_SetSpawnGroup(Number of Spawn Groups=5)
   &InitializeParticle(
    Lifetime=&UniformRangedFloat(Minimum=0.5, Maximum=0.75)
    Mesh Scale=&VectorFromFloat(Value=&RandomRangeFloat(Minimum=0.04, Maximum=0.05))
    Module.Write Color=true
    Module.Write Lifetime=true
    Module.Write Mass=true
    Module.Write Material Random=true
    Module.Write Position=false
    Module.Write Scale=true
    Module.Write SpriteRotation=false
    Module.Write SpriteSize=true
   )
   &SkeletalMeshLocation(
    Bone / Socket Index=$Particles.SpawnGroup
    Bone Sampling Mode=::NewEnumerator4
    Inherit Velocity=false
    MaskSpawnGroup=false
    Sampled Position Offset Coordinate Space:ENiagaraMeshTransforms=::NewEnumerator3
    Skeletal Mesh=$System.Skeletal Mesh
   )
   &SphereLocation(
    MaskSpawnGroup=false
    Override Seed=false
    Sphere Radius=2
    Surface Only=false
    UseNonUniformScale=false
    UseOffset=false
    UseSphereOrientation=false
   )
   &AddVelocity(Velocity=&Multiply_VectorByFloat(
     Vector=&RandomRangeVector(Minimum=<-3 -3 10>, Maximum=<3 3 30>)
     Float=HlslExpression() {hlslExpression="Particles.SpawnGroup < 3 ? 1:-1"}
    ))
  ]
  particleUpdateScript [
   &ParticleState()
   &ScaleVelocity(Velocity Scale=&MakeVector(
     X=1
     Y=1
     Z=&FloatFromCurve(
      FloatCurve=DataInterface() {propertyValues {Curve {keys [
          {arriveTangent=7.4745, interpMode=RCIM_Cubic, leaveTangent=7.4745, tangentMode=RCTM_User, time=0, value=0}
          {arriveTangent=14.58, interpMode=RCIM_Cubic, leaveTangent=14.58, tangentMode=RCTM_User, time=1, value=4}
         ]}}}
      Scale Curve=&MakeCustomFloatFromBool(Bool=$User.bisFootHidden, True Float=0, False Float=1)
     )
    ))
   &InheritSourceMovement(
    Applied Position Scale=&Multiply_VectorByFloat(
     Vector=<1 1 1>
     Float=&FloatFromCurve(FloatCurve=DataInterface() {propertyValues {Curve {keys [<at 0.4 1>,<at 1 0>]}}})
    )
    ApplyForce=false
    ApplyPosition=true
    ApplyPositionFalloff=false
   )
   &SolveForcesAndVelocity(Clamp Velocity=false, Limit Acceleration=false)
   &ScaleMeshSize(Scale Factor=&Multiply_VectorByFloat(
     Vector=&VectorFromFloat(Value=&FloatFromCurve(FloatCurve=DataInterface() {propertyValues {Curve {keys [
           {arriveTangent=5.6263, interpMode=RCIM_Cubic, leaveTangent=5.6263, tangentMode=RCTM_User, time=0, value=0}
           {interpMode=RCIM_Cubic, tangentMode=RCTM_User, time=0.5, value=1}
           {arriveTangent=-4.9972, interpMode=RCIM_Cubic, leaveTangent=-4.9972, tangentMode=RCTM_User, time=0.9, value=0}
          ]}}}))
     Float=&MakeCustomFloatFromBool(
      Bool=$User.bisFootHidden
      True Float=&MakeCustomFloatFromBool(
       Bool=&SetBoolByIntComparison(A=$Particles.SpawnGroup, B=2, Comparison Type=::NewEnumerator6)
       True Float=0
       False Float=&MakeCustomFloatFromBool(Bool=$User.bisHandHidden, True Float=0, False Float=1)
      )
      False Float=&MakeCustomFloatFromBool(
       Bool=$User.bisHandHidden
       True Float=&MakeCustomFloatFromBool(
        Bool=&SetBoolByIntComparison(A=$Particles.SpawnGroup, B=2, Comparison Type=::NewEnumerator6)
        True Float=1
        False Float=0
       )
       False Float=1
      )
     )
    ))
  ]
  rendererClasses [&NiagaraMeshRendererProperties]
  renderers [{rendererClass=&NiagaraMeshRendererProperties, rendererIndex=0}]
 }
 Upwards_Drips {
  emitterUpdateScript [
   &EmitterState()
   &SpawnBurst_Instantaneous(
    Spawn Count=&RandomRangeInt(Minimum=1, Maximum=2)
    Spawn Probability=HlslExpression() {hlslExpression="Emitter.LoopCount == 0"}
    Use Spawn Probability=false
   )
  ]
  particleSpawnScript [
   &FN_SetSpawnGroup(Number of Spawn Groups=3)
   /SetVariables(
    Particles.PendulumDrag=1
    Particles.PendulumLengthScale:NiagaraFloat=&MaskFloatBySpawnGroup(
     Masked Float=&RandomRangeFloat(Minimum=0.5, Maximum=0.535)
     Passthrough Float=&RandomRangeFloat(Minimum=0.35, Maximum=0.4)
     Spawn Group Mask=2
    )
    Particles.PendulumMaskFalloff=1
    Particles.PendulumMeshWidth:NiagaraFloat=&RandomRangeFloat(Minimum=0.1, Maximum=0.175)
    Particles.PendulumRigidity=0.06
   ) {isSetParameters=true, module=SetVariables_#2}
   &InitializeParticle(
    Lifetime=&UniformRangedFloat(Minimum=3, Maximum=6)
    Mesh Scale=&MakeVector(X=$Particles.PendulumMeshWidth, Y=$Particles.PendulumMeshWidth, Z=$Particles.PendulumLengthScale)
    Module.Write Color=true
    Module.Write Lifetime=true
    Module.Write Mass=true
    Module.Write Material Random=true
    Module.Write Position=false
    Module.Write Scale=true
    Module.Write SpriteRotation=false
    Module.Write SpriteSize=true
   )
   &CameraOffset(Camera Offset Amount=2)
  ]
  particleUpdateScript [
   &ParticleState()
   &SkeletalMeshLocation(
    Apply To Particle Position=false
    Bone / Socket Index=$Particles.SpawnGroup
    Bone Sampling Mode=::NewEnumerator4
    Inherit Velocity=false
    MaskSpawnGroup=false
    Skeletal Mesh=$System.Skeletal Mesh
   )
   /SetVariables(Particles.MeshOrigin:Vector3f=&Add_Vector(
     A=$Output.SkeletalMeshLocation.SampledPosition
     B=&RotateVectorByQuaternion(
      Vector To Rotate=&MaskVectorBySpawnGroup(Masked Vector=<2 5 0>, Passthrough Vector=&RandomRangeVector(Minimum=<-2 -2 -2>, Maximum=<2 2 2>))
      Quaternion=$Output.SkeletalMeshLocation.SampledRotation
     )
    )) {isSetParameters=true, module=SetVariables_#3}
   &Drag(Drag=$Particles.PendulumDrag, Use Linear Drag=true, Use Rotational Drag=false)
   &GravityForce()
   &PendulumSetup(
    Module.Enable Spring Driver=true
    Pendulum Length:NiagaraFloat=&Multiply_Float(A=100, B=$Particles.PendulumLengthScale)
    Pendulum Pivot:Vector3f=$Particles.MeshOrigin
    Pendulum Rest Axis:Vector3f=<0 0 1>
    "Spring Driven Constraint (Tightness)":NiagaraFloat=&MaskFloatBySpawnGroup(Masked Float=50, Passthrough Float=12, Spawn Group Mask=2)
   )
   &SolveForcesAndVelocity(Clamp Velocity=false, Limit Acceleration=false)
   &PendulumConstraint()
   &CalculateAccurateVelocity()
   /SetVariables(Particles.MeshFacing:Vector3f=&NormalizeVector(Vector To Normalize=&Subtract_Vector(A=$Particles.MeshOrigin, B=Linked() {linkedVariable {name=Particles.Position, type=NiagaraPosition}}))) {isSetParameters=true, module=SetVariables_#4}
   &self:Pendulum_Facing(
    Bone Quat:Quat4f=$Output.SkeletalMeshLocation.SampledRotation
    FromVector:Vector3f=<0 0 -1>
    ToVector:Vector3f=$Particles.MeshFacing
    X Vector:Vector3f=<-1 0 0>
    Y Vector:Vector3f=<0 0 0>
   ) {module=Pendulum_Facing}
   &DynamicMaterialParameters(
    Index 1 Param 3=1
    Index 1 Param 4=&FloatFromCurve(FloatCurve=DataInterface() {propertyValues {Curve {keys [
         {arriveTangent=3.6645, interpMode=RCIM_Cubic, leaveTangent=3.6645, tangentMode=RCTM_User, time=0.1, value=-0.1}
         {interpMode=RCIM_Cubic, leaveTangent=2.3573, time=0.9, value=2}
        ]}}})
    Index 2 Param 1=HlslExpression() {hlslExpression=Particles.MeshOrigin.x}
    Index 2 Param 2=HlslExpression() {hlslExpression=Particles.MeshOrigin.y}
    Index 2 Param 3=HlslExpression() {hlslExpression=Particles.MeshOrigin.z}
    Index 2 Param 4=$Particles.PendulumMaskFalloff
    Index 3 Param 1=&MakeFloatFromVector(Vector=Linked() {linkedVariable {name=Particles.Position, type=NiagaraPosition}}, Channel=::NewEnumerator0)
    Index 3 Param 2=&MakeFloatFromVector(Vector=Linked() {linkedVariable {name=Particles.Position, type=NiagaraPosition}}, Channel=::NewEnumerator1)
    Index 3 Param 3=&MakeFloatFromVector(Vector=Linked() {linkedVariable {name=Particles.Position, type=NiagaraPosition}}, Channel=::NewEnumerator2)
    Index 3 Param 4=$Particles.PendulumRigidity
    Module.Param0WriteEnabled1=true
    Module.Param0WriteEnabled2=true
    Module.Param0WriteEnabled3=true
    Module.Param1WriteEnabled1=true
    Module.Param1WriteEnabled2=true
    Module.Param1WriteEnabled3=true
    Module.Param2WriteEnabled1=true
    Module.Param2WriteEnabled2=true
    Module.Param2WriteEnabled3=true
    Module.Param3WriteEnabled1=true
    Module.Param3WriteEnabled2=true
    Module.Param3WriteEnabled3=true
    Write Parameter Index 0=false
    Write Parameter Index 1=true
    Write Parameter Index 2=true
    Write Parameter Index 3=true
   )
   &ScaleMeshSize(Scale Factor=&Multiply_VectorByFloat(
     Vector=&MakeVector(
      X=&FloatFromCurve(FloatCurve=DataInterface() {propertyValues {Curve {keys [
           <at 0 0 RCIM_Cubic>
           {arriveTangent=1.0833, interpMode=RCIM_Cubic, leaveTangent=1.0833, time=0.05, value=0.9}
           {arriveTangent=-0.6176, interpMode=RCIM_Cubic, leaveTangent=-0.6176, time=0.6, value=0.65}
           {interpMode=RCIM_Cubic, leaveTangent=-4.2857, time=0.9, value=0.375}
           {interpMode=RCIM_Cubic, leaveTangent=-0.6579, time=0.92, value=0}
          ]}}})
      Y=&FloatFromCurve(FloatCurve=DataInterface() {propertyValues {Curve {keys [
           <at 0 0 RCIM_Cubic>
           {arriveTangent=1.0833, interpMode=RCIM_Cubic, leaveTangent=1.0833, time=0.05, value=0.9}
           {arriveTangent=-0.6176, interpMode=RCIM_Cubic, leaveTangent=-0.6176, time=0.6, value=0.65}
           {interpMode=RCIM_Cubic, leaveTangent=-4.2857, time=0.9, value=0.375}
           {interpMode=RCIM_Cubic, leaveTangent=-0.6579, time=0.92, value=0}
          ]}}})
      Z=&FloatFromCurve(FloatCurve=DataInterface() {propertyValues {Curve {keys [
           {arriveTangent=1.8796, interpMode=RCIM_Cubic, time=0, value=0}
           {arriveTangent=-4.7929, interpMode=RCIM_Cubic, leaveTangent=-4.7929, tangentMode=RCTM_User, time=0.625, value=0.9}
           {arriveTangent=1.157, interpMode=RCIM_Cubic, leaveTangent=1.157, tangentMode=RCTM_User, time=0.95, value=0}
          ]}}})
     )
     Float=&MakeCustomFloatFromBool(
      Bool=$User.bisFootHidden
      True Float=&MaskFloatBySpawnGroup(
       Masked Float=&MakeCustomFloatFromBool(Bool=$User.bisHandHidden, True Float=0, False Float=1)
       Passthrough Float=0
       Spawn Group Mask=2
      )
      False Float=&MakeCustomFloatFromBool(
       Bool=$User.bisHandHidden
       True Float=&MaskFloatBySpawnGroup(Masked Float=0, Passthrough Float=1, Spawn Group Mask=2)
       False Float=1
      )
     )
    ))
   /SetVariables(
    Particles.Drop Mesh Size:Vector3f=&VectorFromCurve(
     VectorCurve=DataInterface() {propertyValues {
       XCurve {keys [
         <at 0 1>
         {arriveTangent=-18.2164, interpMode=RCIM_Cubic, leaveTangent=-18.2164, tangentMode=RCTM_User, time=0.85, value=1}
         {arriveTangent=-2.286, interpMode=RCIM_Cubic, leaveTangent=-2.286, tangentMode=RCTM_User, time=0.8575, value=0}
        ]}
       YCurve {keys [
         <at 0 1>
         {arriveTangent=-18.2164, interpMode=RCIM_Cubic, leaveTangent=-18.2164, tangentMode=RCTM_User, time=0.85, value=1}
         {arriveTangent=-2.286, interpMode=RCIM_Cubic, leaveTangent=-2.286, tangentMode=RCTM_User, time=0.8575, value=0}
        ]}
       ZCurve {keys [
         <at 0.1 0.675>
         <at 0.3 0.8>
         <at 0.6 0.4>
         <at 0.8 0.8>
         {interpMode=RCIM_Cubic, leaveTangent=-5.2459, time=0.85, value=0.675}
         <at 0.8575 0>
        ]}
      }}
     Scale Curve=&VectorFromFloat(Value=&FloatFromCurve(
       FloatCurve=DataInterface() {propertyValues {Curve {keys [<at 0.46 0>,<at 0.51 2>,<at 0.6 1>,<at 0.7 0.55>,<at 0.8 1.2>,<at 0.9 0.7>,<at 1 1>]}}}
       Scale Curve=&RandomRangeFloat(Minimum=0.04, Maximum=0.05)
      ))
    )
    Particles.Drop Position:Vector3f=&Lerp_Vector(
     A=$Particles.MeshOrigin
     B=&Add_Vector(
      A=Linked() {linkedVariable {name=Particles.Position, type=NiagaraPosition}}
      B=&Multiply_VectorByFloat(
       Vector=&Multiply_VectorByFloat(
        Vector=&RandomRangeVector(Minimum=<-4 -4 -4>, Maximum=<4 4 4>)
        Float=&Sine(Normalized Angle=$Emitter.Age, Period=1, Scale=1, Bias=0)
       )
       Float=&FloatFromCurve(FloatCurve=DataInterface() {propertyValues {Curve {keys [
            {arriveTangent=0.7009, interpMode=RCIM_Cubic, leaveTangent=0.7009, tangentMode=RCTM_User, time=0.55, value=1}
            {arriveTangent=0.7952, interpMode=RCIM_Cubic, leaveTangent=0.7952, tangentMode=RCTM_User, time=0.9, value=0}
           ]}}})
      )
     )
     Alpha=&FloatFromCurve(FloatCurve=DataInterface() {propertyValues {Curve {keys [
          {arriveTangent=-0.2799, interpMode=RCIM_Cubic, leaveTangent=-0.2799, tangentMode=RCTM_User, time=0.275, value=0}
          {interpMode=RCIM_Cubic, tangentMode=RCTM_Break, time=0.6, value=0.35}
          {arriveTangent=5.35, interpMode=RCIM_Cubic, leaveTangent=15.5091, tangentMode=RCTM_Break, time=0.8, value=0.45}
          {arriveTangent=0.846, interpMode=RCIM_Cubic, leaveTangent=0.846, tangentMode=RCTM_User, time=1, value=2}
         ]}}})
    )
   ) {enabled=false, isSetParameters=true, module=SetVariables_#5}
  ]
  rendererClasses [&NiagaraMeshRendererProperties]
  renderers [
   {rendererClass=&NiagaraMeshRendererProperties, rendererIndex=0}
   {rendererClass=&NiagaraMeshRendererProperties, rendererIndex=1}
  ]
 }
}
summary {
 systemName=NS_Accumulate_Idle
 emitters [
  {emitterName=Upwards_Drips, rendererClasses [&NiagaraMeshRendererProperties]}
  {emitterName=Downward_Drips, rendererClasses [&NiagaraMeshRendererProperties]}
  {emitterName=Hero_Drips, rendererClasses [&NiagaraMeshRendererProperties]}
  {emitterName=Hero_Drops, rendererClasses [&NiagaraMeshRendererProperties]}
  {emitterName=Secondary_Drops, rendererClasses [&NiagaraMeshRendererProperties]}
 ]
 userVariables [
  User.bisFootHidden {type=NiagaraBool}
  User.bIsFrontEnd {type=NiagaraBool}
  User.bIsFrontEndPreview {type=NiagaraBool}
  User.bisHandHidden {type=NiagaraBool}
  User.SlurpColorAlpha {type=NiagaraFloat, defaultValue=0, description=""}
 ]
}
