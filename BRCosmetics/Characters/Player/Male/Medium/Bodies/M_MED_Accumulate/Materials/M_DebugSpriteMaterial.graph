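# A translucent, lit debug sprite material in the M_MED_Accumulate male player body cosmetic folder, flagged for skeletal mesh and mesh deformer use.
# It holds a scalar Opacity parameter and a masks-type sample of the temporary head sprite texture T_HeadSprites_TEMP.
# The graph shows no links from either node to each other or to the material outputs.
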Material /BRCosmetics/Characters/Player/Male/Medium/Bodies/M_MED_Accumulate/Materials/M_DebugSpriteMaterial
graph {
 properties {
  bUsedWithMeshDeformer=true
  bUsedWithSkeletalMesh=true
  blendMode=BLEND_Translucent
  pixelDepthOffsetMode=PDOM_Legacy
  usedShadingModels=Default Lit
 }
 nodes [
  ScalarParameter() {parameterName=Opacity}
  TextureSample() {samplerType=SAMPLERTYPE_Masks, texture=/Game/Characters/Player/Male/Medium/Bodies/M_MED_Accumulate/Textures/T_HeadSprites_TEMP}
 ]
}
