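# Material function for the Accumulate character body materials that evaluates a cubic Hermite curve between two points, returning a vector.
# P0 is the normalized offset of world position from Point_01 (object pivot by default), and P1 is world position minus point2.
# Named reroutes feed T, squared and cubed, into basis polynomials weighting both points and two scaled tangents, with Tangeant_02 apparently at the start.
# T defaults to looping time scaled by Speed.
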MaterialFunction /BRCosmetics/Characters/Player/Male/Medium/Bodies/M_MED_Accumulate/Materials/MF_HermiteCurve2Points
graph {nodes [
  n1 = FunctionInput(Preview=WorldPosition().XYZ) {inputName=Worldposition}
  n2 = MaterialFunctionCall() {materialFunction=&ObjectPivotPoint}
  n3 = Normalize(VectorInput=Subtract(A=@n1, B=FunctionInput(Preview=@n2.Object Pivot Location) {inputName=Point_01}))
  n4 = FunctionInput(Preview=VectorParameter() {parameterName=Point_02, defaultValue=<rgba 512 512 512 1>}.RGB) {inputName=point2}
  Add(A=@n3, B=@n4)
  n5 = Subtract(A=Multiply(A=NamedRerouteUsage()) {constB=2}, B=Multiply(A=NamedRerouteUsage()) {constB=3})
  n6 = Subtract(A=NamedRerouteUsage(), B=Multiply(A=NamedRerouteUsage()) {constB=2})
  n7 = NamedRerouteUsage()
  n8 = FunctionInput(Preview=VectorParameter() {parameterName=MO, defaultValue=<rgba -1 1 1 1>}.RGB) {inputName=Tangeant_02}
  n9 = FunctionInput(Preview=ScalarParameter() {defaultValue=2048, parameterName=Tangeantscale_02}) {inputName=TangeantScale_02, inputType=FunctionInput_Scalar}
  n10 = Add(A=Multiply(A=Add(A=@n5), B=NamedRerouteUsage()), B=Multiply(A=Add(A=@n6, B=@n7), B=Multiply(A=@n8, B=@n9)))
  n11 = NamedRerouteUsage()
  n12 = NamedRerouteUsage()
  n13 = Multiply(A=Add(A=Multiply(A=@n11) {constB=-2}, B=Multiply(A=@n12) {constB=3}), B=NamedRerouteUsage())
  n14 = Subtract(A=NamedRerouteUsage(), B=NamedRerouteUsage())
  n15 = Multiply(
   A=FunctionInput(Preview=VectorParameter() {parameterName=M1, defaultValue=<rgba -1 1 1 1>}.RGB) {inputName=Tangeant_01}
   B=FunctionInput(Preview=ScalarParameter() {defaultValue=2048, parameterName=Tangeantscale_01}) {inputName=TangeantScale_01, inputType=FunctionInput_Scalar}
  )
  FunctionOutput(None=Add(A=Add(A=@n10, B=@n13), B=Multiply(A=@n14, B=@n15)))
  n16 = Time()
  n17 = ScalarParameter() {defaultValue=1, parameterName=Speed}
  n18 = FunctionInput(Preview=Frac(None=Multiply(A=@n16, B=@n17))) {inputName="F(Time)"}
  n19 = Multiply(A=@n18, B=@n18)
  NamedRerouteDeclaration(None=Multiply(A=@n19, B=@n18)) {name=T3, nodeColor=<rgba 1 0 0.3516 1>}
  NamedRerouteDeclaration(None=@n19) {name=T2, nodeColor=<rgba 1 0 0.7031 1>}
  NamedRerouteDeclaration(None=Subtract(A=@n1, B=@n4)) {name=P1, nodeColor=<rgba 0 1 0.8828 1>}
  NamedRerouteDeclaration(None=@n3) {name=P0, nodeColor=<rgba 1 0.0703 0 1>}
  NamedRerouteDeclaration(None=@n18) {name=T, nodeColor=<rgba 0.7109 1 0 1>}
  Transform(None=@n3) {transformSourceType=TRANSFORMSOURCE_World, transformType=TRANSFORM_Local}
  VectorParameter() {parameterName=Point_01, defaultValue=<rgba 0 0 0 1>}
 ]}
