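# Translucent unlit material for the Accumulate character body's sprites, revealing head flipbook faces as the kill count rises, with a looping front-end preview.
# Vertex red picks each sprite's flipbook cell and kill threshold; green varies its size.
# World position offset expands camera-facing quads with curve-driven pops, curl-noise jiggle, stretch and camera offset.
# Noise distorts UVs, curve atlases color sprites, and a shading-model ID mask, or depth on mobile, clips them.
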Material /BRCosmetics/Characters/Player/Male/Medium/Bodies/M_MED_Accumulate/Materials/M_MED_Accumulate_Sprites
graph {
 properties {
  bUseMaterialAttributes=true
  bUsedWithMeshDeformer=true
  bUsedWithMorphTargets=true
  bUsedWithSkeletalMesh=true
  blendMode=BLEND_Translucent
  pixelDepthOffsetMode=PDOM_Legacy
  shadingModel=MSM_Unlit
  twoSided=true
  usedShadingModels=Unlit
 }
 nodes [
  n1 = MaterialFunctionCall() {materialFunction=/Engine/Functions/Engine_MaterialFunctions02/PivotPainter/PivotPainter_PerObjectData}
  n2 = WorldPosition()
  n3 = MaterialFunctionCall!(
   WorldPosition=Subtract(A=@n1.Pivot Position, B=TransformPosition(None=@n2.XYZ) {transformSourceType=TRANSFORMPOSSOURCE_World})
   PivotLocation=@n1.Pivot Position
   WS Normals
   Custom Object Basis 1
   Custom Object Basis 2
   Custom Object Basis 3
  ) {materialFunction=/Engine/Functions/Engine_MaterialFunctions02/WorldPositionOffset/AlignMeshToTheCamera}
  n4 = VertexColor()
  n5 = LinearInterpolate(
   A=ScalarParameter() {defaultValue=12, parameterName=Face_MinScale}
   B=ScalarParameter() {defaultValue=25, parameterName=Face_MaxScale}
   Alpha=@n4.G
  )
  n6 = OneMinus(None=@n4.R)
  n7 = OneMinus(None=Divide(A=ScalarParameter() {defaultValue=10, parameterName=HC_KillCount}) {constB=10})
  n8 = Clamp(None=Ceil(None=Subtract(A=@n6, B=@n7)))
  Add(
   A=@n3.World Position Offset
   B=Multiply(A=Subtract(A=@n3.Undeformed Pixel Shader World Position, B=@n1.Pivot Position), B=LinearInterpolate(B=@n5, Alpha=@n8))
  )
  n9 = MaterialFunctionCall(Float2=Subtract(A=TextureCoordinate()) {constB=0.5}) {materialFunction=&BreakOutFloat2Components}
  n10 = Multiply(A=@n9.R, B=Transform(None=Constant3Vector() {constant=<rgba 1 0 0 0>}) {transformSourceType=TRANSFORMSOURCE_Camera})
  n11 = LinearInterpolate() {constA=12, constB=12}
  n12 = Multiply(A=@n9.G, B=Transform(None=Constant3Vector() {constant=<rgba 0 -1 0 0>}) {transformSourceType=TRANSFORMSOURCE_Camera})
  Add(
   A=Add(A=Multiply(A=@n10, B=@n11), B=Multiply(A=@n12, B=@n11))
   B=MaterialFunctionCall(Offset Amount=ScalarParameter() {defaultValue=21, parameterName="Offset Amount (S)"}) {materialFunction=/Engine/Functions/Engine_MaterialFunctions02/WorldPositionOffset/CameraOffset}.Clamped Camera Offset
  )
  n13 = PreSkinnedPosition()
  n14 = MaterialFunctionCall() {materialFunction=&PivotPainter_PerObjectData}
  n15 = ScalarParameter() {defaultValue=32, parameterName=CurlNoise_Tiling}
  n16 = Time()
  n17 = VectorParameter() {parameterName=Sprites_DriftSpeed, defaultValue=<rgba 0.1 0.25 0.17 1>}
  n18 = ScalarParameter() {parameterName=BlendTime}
  n19 = Add(A=Add(A=Divide(A=@n13, B=@n15), B=Multiply(A=@n16, B=@n17.RGB)), B=Multiply(A=@n18) {constB=0.1})
  n20 = VectorParameter() {parameterName=Sprites_JiggleScale, defaultValue=<rgba 0.5 0.5 0.5 1>}
  n21 = VectorParameter() {parameterName=Elim_Sprites_JiggleScale, defaultValue=<rgba 2 2 2 1>}
  n22 = Add(
   A=Normalize(VectorInput=Subtract(A=@n13, B=@n14.Pivot Position))
   B=Multiply(A=VectorNoise(World Position=@n19) {noiseFunction=VNF_CurlALU}, B=LinearInterpolate(A=@n20.RGB, B=@n21.RGB, Alpha=@n18))
  )
  Add(A=NamedRerouteUsage(), B=@n22)
  n23 = MaterialFunctionCall() {materialFunction=&ObjectPivotPoint}
  Add(A=@n23.Object Pivot Location, B=Constant3Vector() {constant=<rgba 0 0 -128 0>}) {constB=92}
  Add(A=@n14.Pivot Position, B=@n23.Object Pivot Location)
  AppendVector(A=ScalarParameter() {defaultValue=2, parameterName=RowSelector}, B=ScalarParameter() {defaultValue=2, parameterName=ColumnSelector})
  Constant2Vector() {g=2, r=2}
  Constant3Vector() {constant=<rgba 512 512 512 0>}
  Constant3Vector() {constant=<rgba 512 512 -512 0>}
  n24 = VertexColor()
  ConstantBiasScale(None=MaterialFunctionCall(Float=@n24.R) {materialFunction=&MF_Hash}.Float)
  Constant() {r=1}
  Fresnel()
  n25 = VertexColor()
  n26 = OneMinus(None=@n25.R)
  n27 = ScalarParameter() {defaultValue=10, parameterName=HC_KillCount}
  n28 = Time()
  n29 = ScalarParameter() {defaultValue=-100000000, group=TargetPractice, parameterName=HC_FEGameTimeFromBP}
  n30 = Multiply(A=Subtract(A=Subtract(A=@n28, B=@n29)) {constB=0.5}, B=ScalarParameter() {defaultValue=0.1, parameterName=FE_LoopSpeed})
  n31 = Multiply(A=@n27, B=Frac(None=@n30))
  n32 = ScalarParameter() {parameterName=IsInFrontEnd}
  n33 = Divide(A=LinearInterpolate(A=@n27, B=Ceil(None=@n31), Alpha=@n32)) {constB=10}
  n34 = OneMinus(None=@n33)
  n35 = Clamp(None=Ceil(None=Subtract(A=@n26, B=@n34)))
  LinearInterpolate(A=ScalarParameter() {defaultValue=3, parameterName=MinMip}, B=ScalarParameter() {parameterName=MaxMip}, Alpha=@n35)
  n36 = ScalarParameter() {defaultValue=0.005, parameterName=DistortionStrength}
  n37 = ComponentMask(None=@n24.R) {r=true}
  n38 = Multiply(
   A=Time()
   B=LinearInterpolate(
    A=ScalarParameter() {defaultValue=0.015, parameterName=HC_SpritePulseSpeed}
    B=ScalarParameter() {defaultValue=0.1, parameterName=HC_SpriteReactPulseSpeed}
    Alpha=ScalarParameter() {parameterName=MaxKillsReached}
   )
  )
  n39 = Frac(None=Add(A=@n37, B=@n38))
  n40 = Constant() {r=0.5}
  n41 = Constant() {r=0.1}
  n42 = Constant()
  n43 = Frac(None=Add(A=@n37, B=Add(A=@n38) {constB=0.75}))
  n44 = Clamp(None=Max(A=SphereMask(A=@n39, B=@n40, Radius=@n41, Hardness=@n42), B=SphereMask(A=@n43, B=@n40, Radius=@n41, Hardness=@n42)))
  n45 = SphereMask(A=@n25.R, B=@n33, Radius=Constant() {r=0.1}, Hardness=Constant() {r=1})
  n46 = Subtract(A=Time(), B=ScalarParameter() {defaultValue=-100000000, group=TargetPractice, parameterName=HC_GameTimeFromBP})
  n47 = ScalarParameter() {defaultValue=1, group=TargetPractice, parameterName=HC_Pop-InDuration}
  n48 = LinearInterpolate(A=Saturate(None=Divide(A=@n46, B=@n47)), B=Frac(None=Multiply(A=@n30) {constB=10}), Alpha=@n32)
  n49 = CurveAtlasRowParameter(CurveTime=LinearInterpolate(A=@n44, B=@n48, Alpha=@n45)) {
   atlas=&AthenaCurveAtlas01
   curve=/Game/Characters/Player/Male/Medium/Bodies/M_MED_Accumulate/Materials/AccumulateSpriteAnimCurve
   defaultValue=235
   parameterName=PopInCurve
  }
  n50 = Multiply(A=@n45, B=@n49.R)
  LinearInterpolate(A=LinearInterpolate(A=@n36, Alpha=@n44) {constA=0.03, constB=0}, B=LinearInterpolate(Alpha=@n50) {constA=0.01, constB=0}, Alpha=@n50) {constB=0}
  n51 = VectorParameter() {parameterName=Color_02, defaultValue=<rgba 0 0.4993 0.535 1>}
  n52 = VectorParameter() {parameterName=Color_01, defaultValue=<rgba 1 1 1 1>}
  n53 = TextureCoordinate()
  n54 = AppendVector(A=ScalarParameter() {defaultValue=1, parameterName=Noise_X_Tiling}, B=ScalarParameter() {defaultValue=1, parameterName=Noise_Y_Tiling})
  n55 = ScalarParameter() {parameterName=Noise_X_Speed}
  n56 = ScalarParameter() {defaultValue=0.5, parameterName=Noise_Y_Speed}
  n57 = TextureSample(UVs=Panner(Coordinate=Multiply(A=@n53, B=@n54), Speed=AppendVector(A=@n55, B=@n56))) {texture=/Game/Characters/Player/Male/Medium/Bodies/M_MED_Viper/Textures/Glitch/T_SoftNoise}
  n58 = ComponentMask(None=Subtract(A=Multiply(A=@n57.RGB) {constB=2})) {g=true, r=true}
  n59 = Multiply(A=TextureCoordinate(), B=Divide(A=Constant() {r=1}, B=ScalarParameter() {defaultValue=3, parameterName=RowAndColumns}))
  n60 = Constant2Vector()
  n61 = VertexColor()
  n62 = Divide(A=Constant() {r=2}) {constB=10}
  n63 = LinearInterpolate(A=@n60, B=Constant2Vector() {g=1}, Alpha=Ceil(None=SphereMask(A=@n61.R, B=@n62) {attenuationRadius=0.09, hardnessPercent=1}))
  n64 = Constant2Vector() {g=2}
  n65 = Constant() {r=3}
  n66 = Ceil(None=SphereMask(A=@n61.R, B=Divide(A=@n65) {constB=10}) {attenuationRadius=0.09, hardnessPercent=1})
  n67 = SphereMask(A=@n61.R, B=Divide(A=Constant() {r=4}) {constB=10}) {attenuationRadius=0.09, hardnessPercent=1}
  n68 = Divide(A=Constant() {r=5}) {constB=10}
  n69 = LinearInterpolate(
   A=LinearInterpolate(A=LinearInterpolate(A=@n63, B=@n64, Alpha=@n66), B=Constant2Vector() {r=1}, Alpha=Ceil(None=@n67))
   B=Constant2Vector() {g=1, r=1}
   Alpha=Ceil(None=SphereMask(A=@n61.R, B=@n68) {attenuationRadius=0.09, hardnessPercent=1})
  )
  n70 = Constant2Vector() {g=2, r=1}
  n71 = Constant() {r=6}
  n72 = Ceil(None=SphereMask(A=@n61.R, B=Divide(A=@n71) {constB=10}) {attenuationRadius=0.09, hardnessPercent=1})
  n73 = Constant() {r=7}
  n74 = Ceil(None=SphereMask(A=@n61.R, B=Divide(A=@n73) {constB=10}) {attenuationRadius=0.09, hardnessPercent=1})
  n75 = Divide(A=Constant() {r=8}) {constB=10}
  n76 = LinearInterpolate(
   A=LinearInterpolate(A=LinearInterpolate(A=@n69, B=@n70, Alpha=@n72), B=Constant2Vector() {r=2}, Alpha=@n74)
   B=Constant2Vector() {g=1, r=2}
   Alpha=Ceil(None=SphereMask(A=@n61.R, B=@n75) {attenuationRadius=0.09, hardnessPercent=1})
  )
  n77 = Constant2Vector() {g=2, r=2}
  n78 = Constant() {r=9}
  n79 = Ceil(None=SphereMask(A=@n61.R, B=Divide(A=@n78) {constB=10}) {attenuationRadius=0.09, hardnessPercent=1})
  n80 = SphereMask(A=@n61.R, B=Divide(A=Constant() {r=10}) {constB=10}) {attenuationRadius=0.09, hardnessPercent=1}
  n81 = Divide(A=LinearInterpolate(A=LinearInterpolate(A=@n76, B=@n77, Alpha=@n79), B=@n60, Alpha=Ceil(None=@n80))) {constA=3, constB=3}
  n82 = TextureSampleParameter2D(UVs=Add(A=Multiply(A=@n58, B=@n36), B=Add(A=@n59, B=@n81))) {
   parameterName=SpriteFlipBook
   samplerType=SAMPLERTYPE_Masks
   texture=/Game/Characters/Player/Male/Medium/Bodies/M_MED_Accumulate/Textures/T_HeadSprites_TEMP
  }
  LinearInterpolate(
   A=VectorParameter() {parameterName=Sprite_BG_Color, defaultValue=<rgba 0 0 0 1>}.RGB
   B=LinearInterpolate(A=@n51.RGB, B=@n52.RGB, Alpha=@n82.G)
   Alpha=@n82.G
  )
  n83 = NamedRerouteUsage()
  LinearInterpolate(B=LinearInterpolate(A=NamedRerouteUsage(), B=NamedRerouteUsage(), Alpha=@n83), Alpha=@n83)
  n84 = CurveAtlasRowParameter(CurveTime=@n82.G) {
   atlas=&AthenaCurveAtlas01
   curve=/Game/Characters/Player/Male/Medium/Bodies/M_MED_Accumulate/Materials/AccumulateBlackColorCurve
   defaultValue=276
   parameterName=FaceColorCurve
  }
  n85 = ScalarParameter() {defaultValue=1, parameterName=SpriteEmissiveStrength}
  n86 = LinearInterpolate(A=@n84, B=Power(Base=Multiply(A=@n84, B=@n85), Exp=ScalarParameter() {defaultValue=6, parameterName=SpriteSharpness}), Alpha=@n35)
  n87 = LinearInterpolate(A=Multiply(A=@n49.G, B=ScalarParameter() {defaultValue=0.05, parameterName=HC_BGSpriteBrightness}), B=@n35, Alpha=@n35) {constA=1}
  n88 = CurveAtlasRowParameter(CurveTime=@n83) {
   atlas=&AthenaCurveAtlas01
   curve=&AccumulateBlackColorCurve
   defaultValue=276
   parameterName=GradientColorCurve
  }
  n89 = SphereMask(A=ComponentMask(None=SceneTexture() {sceneTextureId=PPI_ShadingModelID}.Color) {r=true}, B=Constant() {r=4}) {attenuationRadius=0.1}
  n90 = PixelDepth()
  n91 = SceneDepth()
  n92 = Subtract(A=Multiply(A=Subtract(A=@n90.R, B=@n91.R)) {constB=-1}, B=ScalarParameter() {defaultValue=34, parameterName=DepthBias})
  n93 = Clamp(None=Divide(A=@n92, B=ScalarParameter() {defaultValue=1, parameterName=Depth}))
  n94 = ShadingPathSwitch(Default=@n89, Deferred=@n89, Forward=@n89, Mobile=Clamp(None=OneMinus(None=@n93)))
  n95 = Multiply(A=LinearInterpolate(A=Multiply(A=@n86, B=@n87), B=@n88, Alpha=@n83), B=@n94)
  n96 = Multiply(A=@n82.R, B=@n87)
  n97 = Clamp(None=Multiply(A=Multiply(A=@n96) {constB=2}, B=@n94))
  n98 = MaterialFunctionCall(Float2=Subtract(A=TextureCoordinate()) {constB=0.5}) {materialFunction=&BreakOutFloat2Components}
  n99 = Constant3Vector() {constant=<rgba 1 0 0 0>}
  n100 = LinearInterpolate(
   A=ScalarParameter() {defaultValue=12, parameterName=Face_MinScale}
   B=ScalarParameter() {defaultValue=25, parameterName=Face_MaxScale}
   Alpha=@n25.G
  )
  n101 = Multiply(A=@n100, B=@n50)
  n102 = Multiply(A=@n44, B=@n49.G)
  n103 = LinearInterpolate(
   A=ScalarParameter() {defaultValue=15, parameterName=RandomPop_MinScale}
   B=ScalarParameter() {defaultValue=35, parameterName=RandomPop_MaxScale}
   Alpha=@n25.G
  )
  n104 = Add(A=LinearInterpolate(A=Multiply(A=@n102, B=@n103), B=LinearInterpolate(A=@n100, B=@n101, Alpha=@n45), Alpha=@n35), B=NamedRerouteUsage())
  n105 = Multiply(A=Multiply(A=@n98.R, B=Transform(None=@n99) {transformSourceType=TRANSFORMSOURCE_Camera}), B=@n104)
  n106 = Constant3Vector() {constant=<rgba 0 -1 0 0>}
  n107 = VertexColor()
  n108 = Add(
   A=Time()
   B=ConstantBiasScale(None=MaterialFunctionCall(Float=@n107.R) {materialFunction=&MF_Hash}.Float)
  )
  n109 = ScalarParameter() {defaultValue=0.5, parameterName=HC_StretchSpeed}
  n110 = ConstantBiasScale(None=Sine(None=Multiply(A=@n108, B=@n109)))
  n111 = LinearInterpolate(
   A=ScalarParameter() {defaultValue=3, parameterName=HC_Sprites StretchMin}
   B=ScalarParameter() {defaultValue=6, parameterName=HC_Sprites StretchMax}
   Alpha=@n107.G
  )
  n112 = Multiply(A=Multiply(A=@n98.G, B=Transform(None=@n106) {transformSourceType=TRANSFORMSOURCE_Camera}), B=Add(A=Multiply(A=@n110, B=@n111), B=@n104))
  n113 = ScalarParameter() {defaultValue=21, parameterName="Offset Amount (S)"}
  n114 = Add(
   A=ShadingPathSwitch(Default=@n113, Deferred=@n113, Forward=@n113, Mobile=ScalarParameter() {defaultValue=28, parameterName="MobileOffset Amount (S)"})
   B=NamedRerouteUsage()
  )
  n115 = Add(
   A=Add(
    A=Add(A=@n105, B=@n112)
    B=MaterialFunctionCall(Offset Amount=@n114) {materialFunction=&CameraOffset}.Clamped Camera Offset
   )
   B=@n22
  )
  MaterialFunctionCall(In=SetMaterialAttributes!(
    MaterialAttributes
    Emissive Color=NamedRerouteDeclaration(None=@n95) {name=emissive, nodeColor=<rgba 1 0.8906 0 1>}
    Opacity=NamedRerouteDeclaration(None=@n97) {name=opacity, nodeColor=<rgba 1 0.6094 0 1>}
    World Position Offset=NamedRerouteDeclaration(None=@n115) {name=wpo, nodeColor=<rgba 1 0.3984 0 1>}
   ) {attributeSetTypes [#1, #2, #3]}) {materialFunction=&MF_TechArtOverrides}
  n116 = Abs(None=MaterialFunctionCall() {materialFunction=&CameraDirectionVector}.Result)
  n117 = DotProduct(A=ComponentMask(None=SceneTexture() {sceneTextureId=PPI_WorldNormal}.Color) {b=true, g=true, r=true}, B=OneMinus(None=@n116))
  Multiply(A=@n89, B=LinearInterpolate(Alpha=Clamp(None=@n117)))
  n118 = Time()
  Multiply(A=@n118)
  Multiply(A=@n116) {constB=-1}
  n119 = SceneTexture() {sceneTextureId=PPI_WorldNormal}
  n120 = Power(Base=Fresnel(
    ExponentIn=ScalarParameter() {defaultValue=1, parameterName=SilhouetteFresnelExponent}
    Normal=ComponentMask(None=@n119.Color) {b=true, g=true, r=true}
   )) {constExponent=1.5}
  n121 = ScalarParameter() {defaultValue=32, parameterName=SpriteFresnelDepth}
  NamedRerouteDeclaration(None=ShadingPathSwitch(Default=@n120, Deferred=@n120, Forward=@n120, Mobile=Divide(A=@n92, B=@n121))) {name=DepthGradient, nodeColor=<rgba 0.2656 0 1 1>}
  NamedRerouteDeclaration(None=@n51.RGB) {name=Color2, nodeColor=<rgba 0 0.3906 1 1>}
  NamedRerouteDeclaration(None=@n52.RGB) {name=Color1, nodeColor=<rgba 1 0 0.5859 1>}
  NamedRerouteDeclaration(None=@n50) {name=PopInTimeCurve, nodeColor=<rgba 0.8281 1 0 1>}
  n122 = Frac(None=Multiply(A=Time()) {constB=0.5})
  n123 = ScalarParameter() {defaultValue=128, parameterName=TangeantScale}
  n124 = VectorParameter() {parameterName=Tangeants, defaultValue=<rgba -1 -1 -1 1>}
  NamedRerouteDeclaration(None=MaterialFunctionCall!(
    point2=@n14.Pivot Position
    Worldposition=WorldPosition().XYZ
    "F(Time)"=@n122
    TangeantScale_01=@n123
    Tangeant_01=@n124.RGB
    Tangeant_02=@n124.RGB
    TangeantScale_02=@n123
    Point_01=@n14.Pivot Position
   ) {materialFunction=&/MF_HermiteCurve2Points}.Result) {name=HermiteCurveMovement, nodeColor=<rgba 0.5 0 1 1>}
  NamedRerouteDeclaration(None=@n48) {name=KillTime, nodeColor=<rgba 1 0 0.8437 1>}
  n125 = Sine(None=Add(A=VertexColor().R, B=ScalarParameter() {parameterName=ElimCurve}))
  n126 = Constant() {r=0.5}
  n127 = Constant() {r=1}
  n128 = Constant()
  n129 = Subtract(A=Multiply(A=SphereMask(A=@n125, B=@n126, Radius=@n127, Hardness=@n128)) {constB=2})
  n130 = ScalarParameter() {defaultValue=7.5, parameterName=ElimPopScale}
  NamedRerouteDeclaration(None=LinearInterpolate(B=Multiply(A=@n129, B=@n130) {constB=5}, Alpha=ScalarParameter() {parameterName=BlendTime})) {name=ElimPop, nodeColor=<rgba 0 1 0.2969 1>}
  NamedRerouteDeclaration(None=Multiply(A=@n74, B=ScalarParameter() {defaultValue=8, parameterName=ButtClipFix})) {name=ButtClipFix, nodeColor=<rgba 1 0 0.5625 1>}
  NamedRerouteUsage()
  OneMinus(None=@n122)
  PixelNormalWS()
  ScalarParameter() {parameterName=Swirl}
  SceneDepth()
  Sine(None=@n118)
  SphereMask(
   A=ComponentMask(None=SceneTexture() {sceneTextureId=PPI_SubsurfaceColor}.Color) {b=true, g=true, r=true}
   B=VectorParameter() {parameterName=ColorMaskValue, defaultValue=<rgba 0 0.004 0 1>}.RGB
   Radius=Constant() {r=0.002}
   Hardness=Constant() {r=1}
  )
  Subtract(A=Multiply(A=VertexColor()) {constB=2})
  Subtract(A=@n14.Pivot Position, B=@n23.Object Pivot Location)
  Subtract(A=Multiply(A=@n25.R) {constB=2})
  Subtract(A=@n13, B=@n14.Pivot Position)
  WorldPosition()
  WorldPosition()
 ]
}
