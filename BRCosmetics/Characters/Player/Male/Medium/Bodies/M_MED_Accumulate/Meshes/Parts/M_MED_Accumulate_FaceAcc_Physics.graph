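# Physics asset for the M_MED_Accumulate character's FaceAcc mesh part, providing physics bodies for the left and right hoodie drawstrings.
# Each string is a chain of four capsules hanging from a box base body, with both bases constrained to a large spine_05 torso capsule.
# The base-to-spine joints rotate freely, while the links down each chain use tight swing and twist limits.
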PhysicsAsset /BRCosmetics/Characters/Player/Male/Medium/Bodies/M_MED_Accumulate/Meshes/Parts/M_MED_Accumulate_FaceAcc_Physics
bodies {
 dyn_hoodieDrawstring_01_Base_l [{shapeType=Box, extentX=1.5945, extentY=1.4884, extentZ=0.7234}]
 dyn_hoodieDrawstring_01_Base_r [{shapeType=Box, extentX=1.7141, extentY=1.5885, extentZ=0.7234}]
 dyn_hoodieDrawstring_01_l [{shapeType=Capsule, length=2.6603, radius=0.505, center=<2 0 0>, rotation=<rot 90 0 0>}]
 dyn_hoodieDrawstring_01_r [{shapeType=Capsule, length=1.831, radius=0.7212, center=<1.0315 -0.0891 0.0046>, rotation=<rot 90 0 0>}]
 dyn_hoodieDrawstring_02_l [{shapeType=Capsule, length=3.7314, radius=0.6173, center=<2 0 0>, rotation=<rot 90 0 0>}]
 dyn_hoodieDrawstring_02_r [{shapeType=Capsule, length=3.3045, radius=0.5615, center=<2 0 0>, rotation=<rot 90 0 0>}]
 dyn_hoodieDrawstring_03_l [{shapeType=Capsule, length=3.1728, radius=0.6889, center=<2 0 0>, rotation=<rot 90 0 0>}]
 dyn_hoodieDrawstring_03_r [{shapeType=Capsule, length=4.104, radius=0.6771, center=<2 0 0>, rotation=<rot 90 0 0>}]
 dyn_hoodieDrawstring_04_l [{shapeType=Capsule, length=3.3609, radius=0.6659, center=<2 0 0>, rotation=<rot 90 0 0>}]
 dyn_hoodieDrawstring_04_r [{shapeType=Capsule, length=4.8831, radius=0.7807, center=<2.4456 0 0>, rotation=<rot 90 0 0>}]
 spine_05 [{shapeType=Capsule, length=24.1166, radius=15.5245, center=<-15.8239 -2.6628 0.3497>, rotation=<rot -0.1447 -89.3396 -89.2648>}]
}
constraints [
 {
  bone1Name=dyn_hoodieDrawstring_01_Base_r
  bone2Name=spine_05
  swing1LimitDegrees=45
  swing1Motion=Free
  swing2LimitDegrees=45
  swing2Motion=Free
  twistLimitDegrees=45
  twistMotion=Free
 }
 {
  bone1Name=dyn_hoodieDrawstring_01_Base_l
  bone2Name=spine_05
  swing1LimitDegrees=45
  swing1Motion=Free
  swing2LimitDegrees=45
  swing2Motion=Free
  twistLimitDegrees=45
  twistMotion=Free
 }
 {
  bone1Name=dyn_hoodieDrawstring_01_l
  bone2Name=dyn_hoodieDrawstring_01_Base_l
  swing1LimitDegrees=15
  swing2LimitDegrees=15
  twistLimitDegrees=8
 }
 {bone1Name=dyn_hoodieDrawstring_02_l, bone2Name=dyn_hoodieDrawstring_01_l, swing1LimitDegrees=15, swing2LimitDegrees=15, twistLimitDegrees=8}
 {bone1Name=dyn_hoodieDrawstring_03_l, bone2Name=dyn_hoodieDrawstring_02_l, swing1LimitDegrees=15, swing2LimitDegrees=15, twistLimitDegrees=8}
 {bone1Name=dyn_hoodieDrawstring_04_l, bone2Name=dyn_hoodieDrawstring_03_l, swing1LimitDegrees=15, swing2LimitDegrees=15, twistLimitDegrees=8}
 {
  bone1Name=dyn_hoodieDrawstring_01_r
  bone2Name=dyn_hoodieDrawstring_01_Base_r
  swing1LimitDegrees=15
  swing2LimitDegrees=15
  twistLimitDegrees=8
 }
 {bone1Name=dyn_hoodieDrawstring_02_r, bone2Name=dyn_hoodieDrawstring_01_r, swing1LimitDegrees=15, swing2LimitDegrees=15, twistLimitDegrees=8}
 {bone1Name=dyn_hoodieDrawstring_03_r, bone2Name=dyn_hoodieDrawstring_02_r, swing1LimitDegrees=15, swing2LimitDegrees=15, twistLimitDegrees=8}
 {bone1Name=dyn_hoodieDrawstring_04_r, bone2Name=dyn_hoodieDrawstring_03_r, swing1LimitDegrees=15, swing2LimitDegrees=15, twistLimitDegrees=8}
]
