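# Animation Blueprint for the M_MED_Accumulate cosmetic character part, adding secondary motion, including hoodie drawstring setup, to the attached mesh.
# It copies pose and curves from the attached parent, then anchors both drawstring base bones with mirrored, weighted constraints to spine, neck and clavicle bones.
# A RigidBody node then simulates the FaceAcc physics asset in spine_05 space, with overridden gravity and damped, clamped component motion.
# The event graph has no logic.
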AnimBlueprint /BRCosmetics/Characters/Player/Male/Medium/Bodies/M_MED_Accumulate/Meshes/Parts/M_MED_Accumulate_AnimBP
target_skeleton=&/M_MED_Accumulate_Skeleton
graphs [
 AnimGraph {nodes [
   n1 = "Constraint - dyn_hoodieDrawstring_01_Base_l (PPPPPPP)"!(
    ComponentPose=Local To Component(LocalPose=Copy Pose From Mesh() {bCopyCurves=true, bUseAttachedParent=true})
    Alpha=1
    ConstraintWeights_0=0.067
    ConstraintWeights_1:float=0.0842
    ConstraintWeights_2:float=0.3095
    ConstraintWeights_3:float=0.4988
    ConstraintWeights_4:float=0.0077
    ConstraintWeights_5:float=0.0236
    ConstraintWeights_6:float=0.0091
   ) -> (Pose) {
    alphaBoolBlend {blendOption=Linear}
    boneToModify {boneName=dyn_hoodieDrawstring_01_Base_l}
    constraintSetup [
     {targetBone {boneName=clavicle_l}}
     {targetBone {boneName=spine_03}}
     {targetBone {boneName=spine_04}}
     {targetBone {boneName=spine_05}}
     {targetBone {boneName=clavicle_r}}
     {targetBone {boneName=neck_02}}
     {targetBone {boneName=neck_01}}
    ]
    constraintWeights [1,1,1,1,1,1,1]
   }
   n2 = "Constraint - dyn_hoodieDrawstring_01_Base_r (PPPPPPP)"!(
    ComponentPose=@n1
    Alpha=1
    ConstraintWeights_0=0.0072
    ConstraintWeights_1:float=0.0842
    ConstraintWeights_2:float=0.3091
    ConstraintWeights_3:float=0.4997
    ConstraintWeights_4:float=0.0665
    ConstraintWeights_5:float=0.0241
    ConstraintWeights_6:float=0.0092
   ) -> (Pose) {
    alphaBoolBlend {blendOption=Linear}
    boneToModify {boneName=dyn_hoodieDrawstring_01_Base_r}
    constraintSetup [
     {targetBone {boneName=clavicle_l}}
     {targetBone {boneName=spine_03}}
     {targetBone {boneName=spine_04}}
     {targetBone {boneName=spine_05}}
     {targetBone {boneName=clavicle_r}}
     {targetBone {boneName=neck_02}}
     {targetBone {boneName=neck_01}}
    ]
    constraintWeights [1,1,1,1,1,1,1]
   }
   n3 = RigidBody(ComponentPose=@n2) {
    bClampLinearTranslationLimitToRefPose=true
    bOverrideWorldGravity=true
    overridePhysicsAsset=&/M_MED_Accumulate_FaceAcc_Physics
    simulationSpace=BaseBoneSpace
    alphaBoolBlend {blendOption=Linear}
    baseBoneRef {boneName=spine_05}
    componentAppliedLinearAccClamp=<3000 3000 3000>
    componentLinearAccScale {x=0.2}
    componentLinearVelScale {x=0.2}
    overrideWorldGravity {z=-980}
    simSpaceSettings {maxAngularAcceleration=300, maxAngularVelocity=10, maxLinearAcceleration=300, maxLinearVelocity=200, velocityScaleZ=0, worldAlpha=1}
   }
   Output Pose(Result=Component To Local(ComponentPose=@n3)) {layerGroup=None}
  ]}
 EventGraph {nodes [Event BlueprintUpdateAnimation()]}
]
variables []
